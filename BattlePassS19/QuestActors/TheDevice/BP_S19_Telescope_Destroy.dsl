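; Quest prop actor for a destroy-the-telescope objective in Season 19, driven by the player's quest progress.
; On state change, it hides the pre-interaction mesh if the objective is already complete, otherwise runs parent logic and deactivates the ambient effect.
; Visual updates call the parent only for a valid local controller with the objective completed.
; An in-graph note marks this visual update as broken by a quest parent change.
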(event EventChangeStateForPlayer
  (params (FPC :type Fort%20Player%20Controller%20Object%20Reference))
  (bind (_playercontrollerout _completedquestobjective) (Quest|HasPlayerCompletedQuestObjective FPC (Class|ConsolidatedQuestComponent|GetQuestItemDefinition (Variables|BuildingGameplayActorPropQuest|GetConsolidatedQuestComponent)) (Variables|/Script/FortniteGame.ConsolidatedQuestComponent|GetObjectiveBackendName (Variables|BuildingGameplayActorPropQuest|GetConsolidatedQuestComponent))))
  (if _completedquestobjective
    (Development|SetHiddeninGame (Variables|Components|GetSM_Quest_PreInteraction) true true)
    (else
      (|CallParentFunction _playercontrollerout)
      (Components|Activation|Deactivate (Variables|Components|GetP_Ambient)))))

(event EventUpdateVisuals
  (params (PlayerController :type Fort%20Player%20Controller%20Object%20Reference))
  ;[1 0 0] Broken with quest parent change.
  (Utilities|IsValid PlayerController
    (:Is%20Valid
      (if (Pawn|IsLocalPlayerController PlayerController)
        (bind (_playercontrollerout _completedquestobjective) (Quest|HasPlayerCompletedQuestObjective PlayerController (Class|ConsolidatedQuestComponent|GetQuestItemDefinition (Variables|BuildingGameplayActorPropQuest|GetConsolidatedQuestComponent)) (Variables|/Script/FortniteGame.ConsolidatedQuestComponent|GetObjectiveBackendName (Variables|BuildingGameplayActorPropQuest|GetConsolidatedQuestComponent))))
        (if _completedquestobjective
          (|CallParentFunction PlayerController))))))

(fn ConstructionScript
  (|CallParentFunction))
